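# Niagara system for the male Collectable VinderTech glider: socket-attached thruster effects and motion trails on the skeletal mesh.
# Engine_Glow sprites, Flash meshes and flickering Lights sit on thruster sockets, scaling with a startup curve, system timelines and signed User.RightDot, so opposite sides respond to steering.
# FN_AnimTrail spawns eight ribbon trails from further socket or bone indices, enabled and faded by two timelines, with inherited source movement shaping their motion.
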NiagaraSystem /BRCosmetics/Gadgets/Assets/VinderTech_GliderChute/Glider_Collectable_Male/FX/NS_Glider_Collectable_Male
dependencies {
 usedDataInterfaces [
  &NiagaraDataInterfaceArrayFloat
  &NiagaraDataInterfaceMeshRendererInfo
  &NiagaraDataInterfaceCurve
  &NiagaraDataInterfaceArrayFloat3
 ]
 usedDynamicInputs [&ClampFloat, &Abs_Float]
 usedModules [&SystemState, &Timeline]
}
emitters {
 Engine_Glow {
  emitterUpdateScript [
   &EmitterState(
    Inactive Response=::NewEnumerator1
    Life Cycle Mode=::NewEnumerator1
    Loop Behavior=::NewEnumerator1
    Loop Duration Mode=::NewEnumerator1
    UseLoopDelay=true
   )
   &SpawnBurst_Instantaneous(Spawn Count=5, Use Spawn Probability=false)
  ]
  particleSpawnScript [
   &PartitionParticles(Number Of Partitions=5)
   &InitializeParticle(
    Color=<rgba 2 0.1667 0 0.5>
    Sprite Size Mode=::NewEnumerator3
    Uniform Sprite Size=&Multiply_Float(
     A=16
     B=&SelectFloatFromArray(Direct Array Index=$Particles.PartitionParticles.Partition, Float Selection Array=DataInterface() {propertyValues {FloatData [1,0.6]}})
    )
    UsePositionOffset=false
    Write Material Random=true
   )
  ]
  particleUpdateScript [
   &ParticleState(Kill Particles When Lifetime Has Elapsed=false, Loop Particles Lifetime=true)
   &ScaleColor(
    Scale Alpha=&Waveform(
     Global Input Position=&RandomRangeFloat(Minimum=0, Maximum=5)
     "Waveform [1]"=::NewEnumerator12
     "[1] Frequency"=15
     "[1] Amplitude Min/Max"=<0.25 1>
    )
    Scale RGB=&VectorFromFloat(Value=&FN_FrontEndFloatModifier(Gameplay=2, Front End=2))
    ScaleA=true
    ScaleRGB=true
   )
   &SkeletalMeshLocation(
    Bone / Socket Index=$Particles.PartitionParticles.Partition
    MaskSpawnGroup=false
    Mesh Sampling Type=::NewEnumerator5
    Sampled Position Offset=<2 0 0>
    Skeletal Mesh=$User.SkeletalMesh
    Socket Sampling Mode=::NewEnumerator5
   )
   &CameraOffset(Camera Offset Amount=2)
   &ScaleSpriteSize(
    Uniform Curve Index=$Emitter.Age
    Uniform Curve Scale=&Lerp_Float(
     A=1
     B=2
     Alpha=&Max_Float(
      A:NiagaraFloat=$System.Timeline.NormalizedTime
      B:NiagaraFloat=&MakeCustomFloatFromBool(
       Bool=&SetBoolByIntComparison(A=$Particles.PartitionParticles.Partition, B=0, Comparison Type=::NewEnumerator4)
       True Float=0
       False Float=&MakeCustomFloatFromBool(
        Bool=&SetBoolByFloatComparison(
         A=&MakeFloatFromInt(Int=$Particles.PartitionParticles.Partition)
         B=3
         Comparison Type=::NewEnumerator6
        )
        True Float=&Multiply_Float(A=$User.RightDot, B=-1)
        False Float=$User.RightDot
       )
      )
     )
    )
    Uniform Curve Sprite Scale=DataInterface() {propertyValues {Curve {keys [
        {arriveTangent=9.0725, interpMode=RCIM_Cubic, leaveTangent=9.0725, tangentMode=RCTM_User, time=0, value=1}
        {arriveTangent=-2.5249, interpMode=RCIM_Cubic, leaveTangent=-2.5249, tangentMode=RCTM_User, time=0.1408, value=2}
        {arriveTangent=-0.0405, interpMode=RCIM_Cubic, leaveTangent=-0.0405, tangentMode=RCTM_User, time=0.5, value=1}
       ]}}}
   )
  ]
  rendererClasses [&NiagaraSpriteRendererProperties]
  renderers [{rendererClass=&NiagaraSpriteRendererProperties}]
 }
 FN_AnimTrail {
  emitterUpdateScript [
   &EmitterState(Life Cycle Mode=::NewEnumerator1, Loop Behavior=::NewEnumerator1, Loop Duration Mode=::NewEnumerator1, UseLoopDelay=false)
   &FN_SpawnAnimTrail(
    Enable:NiagaraBool=&SetBoolByFloatComparison(
     A=&Max_Float(A:NiagaraFloat=$System.Timeline.NormalizedTime, B:NiagaraFloat=$System.Timeline002.NormalizedTime)
     B=0
    )
    Number of Anim Trails=8
   )
   &FN_SpawnAnimTrail(Number of Anim Trails=8) {enabled=false, module=FN_SpawnAnimTrail001}
  ]
  particleSpawnScript [
   &FN_InitializeAnimTrailBoneIndecies(AnimTrail Selection=7, First Bone or Socket Index=12, Second Bone or Socket Index=12) {module=FN_InitializeAnimTrailBoneIndecies007}
   &FN_InitializeAnimTrailBoneIndecies(AnimTrail Selection=6, First Bone or Socket Index=11, Second Bone or Socket Index=11) {module=FN_InitializeAnimTrailBoneIndecies006}
   &FN_InitializeAnimTrailBoneIndecies(AnimTrail Selection=5, First Bone or Socket Index=10, Second Bone or Socket Index=10) {module=FN_InitializeAnimTrailBoneIndecies005}
   &FN_InitializeAnimTrailBoneIndecies(AnimTrail Selection=4, First Bone or Socket Index=9, Second Bone or Socket Index=9) {module=FN_InitializeAnimTrailBoneIndecies004}
   &FN_InitializeAnimTrailBoneIndecies(AnimTrail Selection=3, First Bone or Socket Index=8, Second Bone or Socket Index=8) {module=FN_InitializeAnimTrailBoneIndecies003}
   &FN_InitializeAnimTrailBoneIndecies(AnimTrail Selection=2, First Bone or Socket Index=7, Second Bone or Socket Index=7) {module=FN_InitializeAnimTrailBoneIndecies002}
   &FN_InitializeAnimTrailBoneIndecies(AnimTrail Selection=1, First Bone or Socket Index=6, Second Bone or Socket Index=6) {module=FN_InitializeAnimTrailBoneIndecies001}
   &FN_InitializeAnimTrailBoneIndecies(First Bone or Socket Index=5, Second Bone or Socket Index=5)
   &FN_InitializeAnimTrail(
    AnimTrailWidth=&Multiply_Float(
     A=5
     B=&SelectFloatFromArray(Direct Array Index=$Particles.SpawnGroup, Float Selection Array=DataInterface() {propertyValues {FloatData [1.5,1.5,1.25,1.25,1,1,1,1]}})
    )
    Color=&MakeLinearColorFromVectorAndFloat(
     "Vector (RGB)"=<1 1 1>
     "Float (Alpha)"=&Multiply_Float(
      A=0.33
      B=&SelectFloatFromArray(
       Direct Array Index=$Particles.SpawnGroup
       Float Selection Array=DataInterface() {propertyValues {FloatData [1,1,0.66,0.66,0.33,0.33,0.33,0.33]}}
      )
     )
    )
    Lifetime=0.4
   )
  ]
  particleUpdateScript [
   &ParticleState()
   &InheritSourceMovement(
    Applied Force Scale=&VectorFromFloat(Value=-1)
    Applied Position Scale=<0.25 0.25 0.25>
    ApplyForce=true
    ApplyForceFalloff=false
    ApplyPosition=true
    ApplyPositionFalloff=false
    Module.LimitForceSpeed=false
   )
   &SolveForcesAndVelocity(Clamp Velocity=false, Limit Acceleration=false)
   &FN_UpdateAnimTrail(Kill Locked particle on system deactivate=false, Locked Particle RibbonWidth=1, SkeletalMesh=$User.SkeletalMesh) {module=FN_UpdateAnimTrail001}
   /SetVariables(Particles.RibbonU0Override=$Particles.NormalizedAge) {isSetParameters=true}
   &ScaleColor(
    Scale Alpha=&FloatFromCurve(
     FloatCurve=DataInterface() {propertyValues {Curve {keys [{interpMode=RCIM_Cubic, tangentMode=RCTM_User, time=0.25, value=1}, {interpMode=RCIM_Cubic, tangentMode=RCTM_User, time=0.66, value=0}]}}}
     Scale Curve=&FloatFromCurve(
      FloatCurve=DataInterface() {propertyValues {Curve {keys [{interpMode=RCIM_Cubic, tangentMode=RCTM_User, time=0.25, value=0}, {interpMode=RCIM_Cubic, tangentMode=RCTM_User, time=1, value=1}]}}}
      CurveIndex=&Max_Float(A:NiagaraFloat=$System.Timeline.NormalizedTime, B:NiagaraFloat=$System.Timeline002.NormalizedTime)
     )
    )
    ScaleA=true
    ScaleRGB=true
   )
  ]
  rendererClasses [&NiagaraRibbonRendererProperties, &NiagaraSpriteRendererProperties]
  renderers [
   {rendererClass=&NiagaraRibbonRendererProperties}
   {rendererClass=&NiagaraSpriteRendererProperties, rendererIndex=1}
  ]
 }
 Flash {
  emitterUpdateScript [
   &EmitterState(
    Inactive Response=::NewEnumerator1
    Life Cycle Mode=::NewEnumerator1
    Loop Behavior=::NewEnumerator1
    Loop Duration Mode=::NewEnumerator1
    UseLoopDelay=true
   )
   &SpawnBurst_Instantaneous(Spawn Count=5, Use Spawn Probability=false)
   &SpawnRate(SpawnRate=50, Use Spawn Probability=false) {enabled=false}
  ]
  particleSpawnScript [
   &PartitionParticles(Number Of Partitions=5)
   &InitializeParticle(
    Lifetime Max=0.133
    Lifetime Min=0.1
    Lifetime Mode=::NewEnumerator1
    Mesh Scale=&Multiply_Vector(
     A=<0.75 0.75 0.75>
     B=&VectorFromFloat(Value=&SelectFloatFromArray(Direct Array Index=$Particles.PartitionParticles.Partition, Float Selection Array=DataInterface() {propertyValues {FloatData [0.1,0.06]}}))
    )
    Mesh Scale Mode=::NewEnumerator4
    UsePositionOffset=false
    Write Material Random=true
   )
   &InitialMeshOrientation(
    Mesh Axis To Orient=<0 0 1>
    Mesh Orientation Mode=::NewEnumerator2
    Orientation Coordinate Space=::Local
    Rotation=&RandomRangeVector(Minimum=<0 0 0>, Maximum=<0 0 1>)
    Use Rotation Vector=true
   )
  ]
  particleUpdateScript [
   &ParticleState(Kill Particles When Lifetime Has Elapsed=false, Loop Particles Lifetime=true)
   &SkeletalMeshLocation(
    Bone / Socket Index=$Particles.PartitionParticles.Partition
    MaskSpawnGroup=false
    Mesh Sampling Type=::NewEnumerator5
    Skeletal Mesh=$User.SkeletalMesh
    Socket Sampling Mode=::NewEnumerator5
   )
   /SetVariables(Particles.ThrusterMask:NiagaraFloat=&Max_Float(
     A:NiagaraFloat=&FloatFromCurve(
      FloatCurve=DataInterface() {propertyValues {Curve {keys [
          {arriveTangent=9.0725, interpMode=RCIM_Cubic, leaveTangent=9.0725, tangentMode=RCTM_User, time=0, value=0}
          <at 0.1408 0.9847 RCIM_Cubic>
          {arriveTangent=-0.0405, interpMode=RCIM_Cubic, leaveTangent=-0.0405, tangentMode=RCTM_User, time=1, value=0}
         ]}}}
      CurveIndex=$Emitter.Age
     )
     B:NiagaraFloat=&Max_Float(
      A:NiagaraFloat=$System.Timeline.NormalizedTime
      B:NiagaraFloat=&MakeCustomFloatFromBool(
       Bool=&SetBoolByIntComparison(A=$Particles.PartitionParticles.Partition, B=0, Comparison Type=::NewEnumerator4)
       True Float=0
       False Float=&MakeCustomFloatFromBool(
        Bool=&SetBoolByFloatComparison(
         A=&MakeFloatFromInt(Int=$Particles.PartitionParticles.Partition)
         B=3
         Comparison Type=::NewEnumerator6
        )
        True Float=&Multiply_Float(A=$User.RightDot, B=-1)
        False Float=$User.RightDot
       )
      )
     )
    )) {isSetParameters=true}
   &ScaleMeshSize(Scale Factor=&Multiply_Vector(
     A=&Lerp_Vector(A=<1 1 1>, B=<1.5 1.5 2.5>, Alpha=$Particles.ThrusterMask)
     B=&Multiply_VectorByFloat(
      Vector=<1 1 1>
      Float=&FloatFromCurve(
       FloatCurve=DataInterface() {propertyValues {Curve {keys [
           {arriveTangent=9.0725, interpMode=RCIM_Cubic, leaveTangent=9.0725, tangentMode=RCTM_User, time=0, value=1}
           <at 0.1408 2 RCIM_Cubic>
           {arriveTangent=-0.0405, interpMode=RCIM_Cubic, leaveTangent=-0.0405, tangentMode=RCTM_User, time=1, value=1}
          ]}}}
       CurveIndex=$System.Timeline001.NormalizedTime
      )
     )
    ))
   &CameraOffset(Camera Offset Amount=2)
  ]
  rendererClasses [&NiagaraMeshRendererProperties]
  renderers [{rendererClass=&NiagaraMeshRendererProperties, rendererIndex=0}]
 }
 Lights {
  emitterUpdateScript [
   &EmitterState(
    Inactive Response=::NewEnumerator1
    Life Cycle Mode=::NewEnumerator1
    Loop Behavior=::NewEnumerator1
    Loop Duration Mode=::NewEnumerator1
    UseLoopDelay=true
   )
   &SpawnBurst_Instantaneous(Use Spawn Probability=false, UseLoopCountLimit=false)
   &SpawnBurst_Instantaneous(Spawn Group=1, Use Spawn Probability=false, UseLoopCountLimit=false) {module=SpawnBurst_Instantaneous002}
   &SpawnBurst_Instantaneous(Spawn Group=3, Use Spawn Probability=false, UseLoopCountLimit=false) {module=SpawnBurst_Instantaneous001}
  ]
  particleSpawnScript [&InitializeParticle(
    Mesh Scale Mode=::NewEnumerator3
    Mesh Uniform Scale=100
    Sprite Size Mode=::NewEnumerator3
    Uniform Sprite Size=100
    UsePositionOffset=false
    Write Material Random=true
   )]
  particleUpdateScript [
   &ParticleState(Kill Particles When Lifetime Has Elapsed=false, Loop Particles Lifetime=true)
   &SkeletalMeshLocation(
    Bone / Socket Index=&SpawnGroup()
    MaskSpawnGroup=false
    Mesh Sampling Type=::NewEnumerator5
    Sampled Position Offset=&SelectVectorFromArray(
     Direct Array Index=&SpawnGroup()
     Vector Selection Array=DataInterface() {propertyValues {FloatData [<10 0 0>, <10 0 -3.0376>]}}
    )
    Skeletal Mesh=$User.SkeletalMesh
    Socket Sampling Mode=::NewEnumerator5
   )
   /SetVariables(Particles.ThrusterMask:NiagaraFloat=&Max_Float(
     A:NiagaraFloat=&FloatFromCurve(
      FloatCurve=DataInterface() {propertyValues {Curve {keys [
          {arriveTangent=9.0725, interpMode=RCIM_Cubic, leaveTangent=9.0725, tangentMode=RCTM_User, time=0, value=0}
          <at 0.1408 0.9847 RCIM_Cubic>
          {arriveTangent=-0.0405, interpMode=RCIM_Cubic, leaveTangent=-0.0405, tangentMode=RCTM_User, time=1, value=0}
         ]}}}
      CurveIndex=$Emitter.Age
     )
     B:NiagaraFloat=&Max_Float(
      A:NiagaraFloat=$System.Timeline.NormalizedTime
      B:NiagaraFloat=&MaskFloatBySpawnGroup(
       Masked Float=0
       Passthrough Float=&MaskFloatBySpawnGroup(Masked Float=$User.RightDot, Passthrough Float=&Multiply_Float(A=$User.RightDot, B=-1), Spawn Group Mask=3)
      )
     )
    )) {isSetParameters=true}
   /SetVariables(Particles.ThrusterIntensity:NiagaraFloat=&Waveform(
     Global Input Position=&RandomRangeFloat(Minimum=0, Maximum=5)
     Global Amplitude Scale=&Lerp_Float(A=1, B=2, Alpha=$Particles.ThrusterMask)
     "Waveform [1]"=::NewEnumerator12
     "[1] Frequency"=15
     "[1] Amplitude Min/Max"=<0.55 1>
    )) {isSetParameters=true}
   /SetVariables(Particles.ThrusterIntensity_Comp:NiagaraFloat=&Multiply_Float(A=20, B=$Particles.ThrusterIntensity)) {isSetParameters=true}
   &FN_Light_Attributes(
    Color=&MakeLinearColorFromVectorAndFloat("Vector (RGB)"=<2 0.5 0>, "Float (Alpha)"=$Particles.ThrusterIntensity)
    Create Color Attribute=true
    Create Exponent Attribute=true
    Create Radius Attribute=true
    Exponent=2
    Radius=&Lerp_Float(A=20, B=30, Alpha=$Particles.ThrusterMask)
   )
  ]
  rendererClasses [&NiagaraLightRendererProperties, &NiagaraComponentRendererProperties]
  renderers [
   {rendererClass=&NiagaraLightRendererProperties, rendererIndex=0}
   {rendererClass=&NiagaraComponentRendererProperties, rendererIndex=1}
  ]
 }
}
summary {
 systemName=NS_Glider_Collectable_Male
 emitters [
  {emitterName=Flash, rendererClasses [&NiagaraMeshRendererProperties]}
  {emitterName=Engine_Glow, rendererClasses [&NiagaraSpriteRendererProperties]}
  {emitterName=FN_AnimTrail, rendererClasses [&NiagaraRibbonRendererProperties, &NiagaraSpriteRendererProperties]}
  {emitterName=Lights, rendererClasses [&NiagaraLightRendererProperties, &NiagaraComponentRendererProperties]}
 ]
 userVariables [
  User.bIsFrontEnd {type=NiagaraBool}
  User.bIsFrontEndPreview {type=NiagaraBool}
  User.ForwardDot {type=NiagaraFloat, defaultValue=0, description=""}
  User.RibbonIDOffset {type=NiagaraInt32}
  User.RightDot {type=NiagaraFloat, defaultValue=0, description=""}
  User.SkeletalMesh {
   type=NiagaraDataInterfaceSkeletalMesh
   defaultValue {
    struct=NiagaraExt_VariableValue_DataInterface
    value {dataInterface=&self:NiagaraDataInterfaceSkeletalMesh_0}
   }
  }
 ]
}
